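# Mannequin body material for AnimToTexture bone animation, usable on instanced static and skeletal meshes.
# A BodyColor-tinted beige plastic is layered with soft metal and black latex through mask texture channels; the green-channel blend mixes plastic with itself.
# A PlasticOverride static switch swaps that stack for a logo plastic layer.
# The result gets a normal map blend, then passes through MF_BoneAnimation as material attributes.
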Material /ProtoSwarm/AnimToTextures/Materials/BoneAnimation/M_Body_BoneAnimation
graph {
 properties {
  bUseMaterialAttributes=true
  bUsedWithInstancedStaticMeshes=true
  bUsedWithMeshDeformer=true
  bUsedWithSkeletalMesh=true
  numCustomizedUVs=1
  pixelDepthOffsetMode=PDOM_Legacy
  usedShadingModels=Default Lit
 }
 nodes [
  n1 = VectorParameter() {parameterName=BodyColor, defaultValue=<rgba 0.451 0.4039 0.3608 0>}
  n2 = MaterialFunctionCall!(LogoColor=@n1.RGB) {materialFunction=/AnimToTexture/Characters/Mannequin/Materials/MaterialFunctions/ML_Plastic_Shiny_Beige_LOGO}
  n3 = MaterialFunctionCall!(PlasticColor=@n1.RGB) {materialFunction=/AnimToTexture/Characters/Mannequin/Materials/MaterialFunctions/ML_Plastic_Shiny_Beige}
  n4 = TextureSample() {samplerType=SAMPLERTYPE_LinearColor, texture=/AnimToTexture/Characters/Mannequin/Textures/UE4_Mannequin_MAT_MASKA}
  n5 = MaterialFunctionCall!(
   Base Material=MaterialFunctionCall(
    Base Material=MaterialFunctionCall(Base Material=@n3.Result, Top Material=@n3.Result, Alpha=@n4.G) {materialFunction=/Engine/Functions/MaterialLayerFunctions/MatLayerBlend_Simple}.Blended Material
    Top Material=MaterialFunctionCall!(Tint, Tiling) {materialFunction=/AnimToTexture/Characters/Mannequin/Materials/MaterialFunctions/ML_SoftMetal}.Result
    Alpha=@n4.R
   ) {materialFunction=&MatLayerBlend_Simple}.Blended Material
   Top Material=MaterialFunctionCall!(Tint, Tiling) {materialFunction=/AnimToTexture/Characters/Mannequin/Materials/MaterialFunctions/ML_GlossyBlack_Latex}.Result
   Alpha=@n4.B
  ) {materialFunction=&MatLayerBlend_Simple}
  n6 = TextureSampleParameter2D() {parameterName=NormalMap, samplerType=SAMPLERTYPE_Normal, texture=/AnimToTexture/Characters/Mannequin/Textures/UE4_Mannequin__normals}
  MaterialFunctionCall!(Material Attributes=MaterialFunctionCall!(NormalMask, Material=StaticSwitchParameter(True=@n2.Result, False=@n5.Blended Material) {parameterName=PlasticOverride}, Normal=@n6.RGB) {materialFunction=/Engine/Functions/MaterialLayerFunctions/MatLayerBlend_NormalBlend}.Blended Material) {materialFunction=/AnimToTexture/Materials/MaterialFunctions/MF_BoneAnimation}
 ]
}
